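# Material function for the CuriousTuna01 Oblong vehicle skin, pairing a blue-grey Fresnel metallic finish with an animated, red-glowing pattern, split by texture alpha.
# Three mirrored, sine-drifting UV1 layers, shifted by OffsetY, are distorted by MF_Noisemaker_Gradient and composited through the texture's green channel.
# The composite feeds AO, emissive, roughness and a pseudo-normal through named reroutes.
# Metal areas use the normal map and gain noise-driven emissive glints.
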MaterialFunction /VehicleCosmetics/Bodies/Oblong/Skins/CuriousTuna01/MF_Skin_Oblong_CuriousTuna01
graph {nodes [
  Constant3Vector()
  n1 = Constant3Vector() {constant=<rgba 0.3763 0.4564 0.6445 1>}
  n2 = TextureCoordinate() {coordinateIndex=1}
  n3 = Add(A=@n2, B=AppendVector(A=Constant(), B=ScalarParameter() {parameterName=OffsetY}))
  n4 = TextureSample(UVs=@n3) {samplerType=SAMPLERTYPE_Normal, texture=&/T_Oblong_CuriousTuna_N}
  n5 = MaterialFunctionCall!(Fresnel Power=Constant() {r=3}, Normal=@n4.RGB) {materialFunction=/VehicleCosmetics/Materials/MaterialFunctions/MF_Fresnel_Prime}
  n6 = Clamp(None=Power(Base=OneMinus(None=@n5.Result)) {constExponent=80, desc=Metallic Finish INT falloff}) {desc=Metallic Dark MinLevel, minDefault=0.02}
  n7 = Multiply(A=@n1, B=Power(Base=@n6) {constExponent=80, desc=Metallic Finish INT falloff})
  n8 = TextureSample(UVs=@n3) {texture=&/T_Oblong_CuriousTuna_RGB}
  n9 = OneMinus(None=@n5.Result)
  n10 = Power(Base=@n5.Result) {constExponent=1.25, desc=Metallic Finish Outer Falloff}
  n11 = Add(A=Multiply(A=@n1, B=Max(A=@n6, B=@n10)), B=@n7)
  n12 = LinearInterpolate(
   A=Max(A=Power(Base=@n9) {constExponent=6, desc=Finish Int Falloff}, B=Power(Base=@n5.Result) {desc=Finish Outer Falloff})
   B=@n11
   Alpha=@n8.A
  )
  n13 = LinearInterpolate(B=@n1, Alpha=Clamp(None=ConstantBiasScale(None=@n8.A) {bias=-0.75, scale=5}))
  n14 = SetMaterialAttributes!(
   MaterialAttributes
   BaseColor=Add(A=LinearInterpolate(B=@n7, Alpha=@n8.A), B=Multiply(A=@n12, B=@n13))
   Metallic=@n8.A
   Roughness=NamedRerouteUsage()
   EmissiveColor=NamedRerouteUsage()
   Normal=NamedRerouteUsage()
   AmbientOcclusion=NamedRerouteUsage()
  ) {attributeSetTypes [
    #1
    #2
    #3
    #4
    #5
    #6
   ]}
  FunctionOutput(None=@n14)
  n15 = Time() {bOverride_Period=true, period=600}
  n16 = TextureCoordinate() {coordinateIndex=1}
  n17 = Abs(None=ConstantBiasScale(None=ComponentMask(None=@n16) {r=true}) {bias=-0.5, scale=2})
  n18 = Add(A=Multiply(A=Sine(None=@n15) {period=7}) {constB=0.02}, B=@n17)
  n19 = Add(A=ComponentMask(None=@n16) {g=true}, B=ScalarParameter() {parameterName=OffsetY})
  n20 = AppendVector(A=Subtract(A=Multiply(A=@n18) {constB=0.9}, B=Multiply(A=@n19) {constB=0.3}), B=Add(A=Multiply(A=@n19) {constB=0.8}) {constB=0.21})
  n21 = MaterialFunctionCall!(Texture Coordinates=@n20, Tiling, PanSpeed=Constant() {r=0.4}, Angle=Constant() {r=40}, RotateSpeed, Sharpen, Offset=Constant() {r=1.5}) {materialFunction=/VehicleCosmetics/Materials/MaterialFunctions/MF_Noisemaker_Gradient}
  n22 = Multiply(A=@n21.Gradient2) {constB=-0.03}
  n23 = Constant()
  n24 = TextureSample(UVs=Add(A=AppendVector(A=@n22, B=@n23), B=@n20)) {texture=&T_Oblong_CuriousTuna_RGB}
  n25 = ConstantBiasScale(None=Multiply(A=@n24.R, B=@n21.Gradient2)) {bias=-0.05, scale=3}
  n26 = Add(A=Multiply(A=Sine(None=@n15) {period=5}) {constB=0.02}, B=@n17)
  n27 = AppendVector(A=Subtract(A=Multiply(A=@n26) {constB=0.875}, B=Multiply(A=@n19) {constB=0.1}), B=Add(A=Multiply(A=@n19) {constB=0.875}) {constB=0.15})
  n28 = MaterialFunctionCall(Texture Coordinates=@n27, PanSpeed=Constant() {r=0.752}, Angle=Constant() {r=60}, Offset=Constant() {r=1}) {materialFunction=&MF_Noisemaker_Gradient}
  n29 = Add(A=AppendVector(A=Multiply(A=@n28.Gradient2) {constB=0.03}, B=Constant()), B=@n27)
  n30 = TextureSample(UVs=@n29) {texture=&T_Oblong_CuriousTuna_RGB}
  n31 = Clamp(None=@n30.R) {maxDefault=0.25}
  n32 = ConstantBiasScale(None=Multiply(A=@n28.Gradient2, B=@n31)) {bias=-0.07, scale=4}
  n33 = Sine(None=@n15) {period=4}
  n34 = AppendVector(A=Add(A=ConstantBiasScale(None=@n33) {scale=0.02}, B=@n17), B=@n19)
  n35 = MaterialFunctionCall(Texture Coordinates=@n34, PanSpeed=Constant() {r=0.5}, Angle=Constant() {r=30}) {materialFunction=&MF_Noisemaker_Gradient}
  n36 = Add(A=AppendVector(A=Multiply(A=@n35.Gradient2) {constB=0.03}, B=Constant()), B=@n34)
  n37 = TextureSample(UVs=@n36) {texture=&T_Oblong_CuriousTuna_RGB}
  n38 = Clamp(None=@n37.R) {maxDefault=0.25}
  n39 = Multiply(A=Clamp(None=ConstantBiasScale(None=@n17) {bias=0.2, scale=3}), B=Multiply(A=@n35.Gradient2, B=@n38))
  n40 = LinearInterpolate(
   A=LinearInterpolate(A=Clamp(None=@n25), B=Clamp(None=@n32), Alpha=@n30.G)
   B=Clamp(None=ConstantBiasScale(None=@n39) {bias=-0.09, scale=6})
   Alpha=@n37.G
  )
  n41 = Add(A=@n29, B=Constant2Vector() {g=0.01, r=-0.02})
  n42 = Subtract(
   A=LinearInterpolate(B=Multiply(A=@n17, B=@n24.R) {constB=0.5}, Alpha=@n24.G)
   B=Multiply(A=TextureSample(UVs=@n41) {texture=&T_Oblong_CuriousTuna_RGB}.R)
  )
  n43 = OneMinus(None=@n28.Gradient2)
  n44 = Multiply(A=Clamp(None=ConstantBiasScale(None=@n43) {bias=0.6, scale=1}), B=Multiply(A=@n17, B=Multiply(A=@n31) {constB=6}))
  n45 = TextureSample(UVs=Add(A=@n36, B=Constant2Vector() {g=0.01, r=-0.05})) {texture=&T_Oblong_CuriousTuna_RGB}
  n46 = ConstantBiasScale(None=OneMinus(None=@n35.Gradient2)) {bias=0.6, scale=1}
  n47 = Multiply(A=@n38) {constB=13}
  n48 = Clamp(None=ConstantBiasScale(None=@n17) {bias=0.075, scale=1})
  n49 = LinearInterpolate(
   A=Subtract(A=LinearInterpolate(A=@n42, B=@n44, Alpha=@n30.G), B=Multiply(A=@n45.G) {constB=0.25})
   B=Multiply(A=Clamp(None=@n46), B=Multiply(A=@n47, B=@n48) {constB=13}) {constB=13}
   Alpha=@n37.G
  )
  n50 = Add(A=@n40, B=@n49)
  NamedRerouteDeclaration(None=LinearInterpolate(A=@n50, Alpha=@n8.A)) {name=AO, nodeColor=<rgba 0.8438 0.8438 0.8438 1>}
  n51 = ConstantBiasScale(None=@n50) {bias=-0.5, scale=0.075}
  n52 = AppendVector(A=@n51, B=@n51)
  n53 = Constant() {r=1}
  NamedRerouteDeclaration(None=LinearInterpolate(A=AppendVector(A=@n52, B=@n53), B=@n4.RGB, Alpha=@n8.A)) {name=Normal, nodeColor=<rgba 0.0026 0.221 1 1>}
  n54 = Constant3Vector() {constant=<rgba 1 0.0343 0.0612 1>}
  n55 = LinearInterpolate(A=@n54, B=Add(A=@n54) {constB=0.2}, Alpha=@n40)
  n56 = Constant3Vector() {constant=<rgba 0 0 1 0>}
  n57 = MaterialFunctionCall(Fresnel Power=Constant() {r=2}, Normal=@n56) {materialFunction=&MF_Fresnel_Prime}
  n58 = ConstantBiasScale(None=Multiply(A=Subtract(A=@n8.A, B=@n8.A), B=@n57.Result)) {bias=-0.3, scale=5}
  n59 = LinearInterpolate(A=LinearInterpolate(A=LinearInterpolate(B=@n55, Alpha=@n50), Alpha=@n8.A) {constB=0}, B=@n11, Alpha=Multiply(A=Clamp(None=@n58)) {constB=3})
  n60 = MaterialFunctionCall(Fresnel Power=Constant() {r=1}, Normal=@n56) {materialFunction=&MF_Fresnel_Prime}
  n61 = ConstantBiasScale(None=@n8.B) {bias=-0.3, scale=2}
  n62 = Multiply(A=Multiply(A=OneMinus(None=@n60.Result), B=Clamp(None=@n61))) {constB=3}
  n63 = Clamp(None=Add(A=@n57.Result) {constB=0.3})
  n64 = Clamp(None=Add(A=@n1) {constB=0.35})
  n65 = ComponentMask(None=@n2) {g=true}
  n66 = Constant() {r=1}
  n67 = MaterialFunctionCall(Texture Coordinates=Clamp(None=Multiply(A=@n8.B, B=@n65)), Tiling=@n66, PanSpeed=@n66, Sharpen=Constant() {r=2}) {materialFunction=&MF_Noisemaker_Gradient}
  n68 = Add(
   A=Multiply(A=ConstantBiasScale(None=@n35.Gradient2) {bias=-0.5, scale=4}, B=@n4.G)
   B=Multiply(A=ConstantBiasScale(None=@n28.Gradient2) {bias=-0.5, scale=4}, B=@n4.R)
  )
  n69 = Clamp(None=Multiply(A=@n8.A, B=Multiply(A=@n67.Gradient3, B=@n68)))
  NamedRerouteDeclaration(None=LinearInterpolate(A=LinearInterpolate(A=@n59, B=@n11, Alpha=@n62), B=Multiply(A=@n63, B=@n64), Alpha=Multiply(A=@n69) {constB=5})) {name=Emissive, nodeColor=<rgba 0 1 0.25 1>}
  n70 = Multiply(A=@n50) {constB=0.75}
  n71 = Add(A=LinearInterpolate(A=Clamp(None=@n70), Alpha=@n8.A) {constB=0}, B=ScalarParameter() {defaultValue=0.25, parameterName=Roughness Minimum}) {constB=0.05}
  n72 = Max(A=@n8.A, B=Ceil(None=@n49))
  NamedRerouteDeclaration(None=Clamp(None=LinearInterpolate(B=@n71, Alpha=@n72) {constA=0.75, constB=0})) {name=Roughness, nodeColor=<rgba 1 0 0.0703 1>}
  OneMinus(None=@n49)
 ]}
