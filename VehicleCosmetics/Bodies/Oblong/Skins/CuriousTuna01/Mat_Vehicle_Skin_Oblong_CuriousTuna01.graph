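# Skeletal mesh vehicle body material for the Oblong CuriousTuna01 skin.
# Blends the shared v2 body base with the skin function, but the alpha subtracts the accent texture's red channel from itself, so only the base shows.
# The result passes through carbon fiber, masked by accent green, then body normal LOD and FN post-FX.
# A separate damage and de-rez chain is not wired into that path.
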Material /VehicleCosmetics/Bodies/Oblong/Skins/CuriousTuna01/Mat_Vehicle_Skin_Oblong_CuriousTuna01
graph {
 properties {
  bAutomaticallySetUsageInEditor=false
  bHasPixelAnimation=true
  bUseMaterialAttributes=true
  bUsedWithMeshDeformer=true
  bUsedWithSkeletalMesh=true
  usedShadingModels=Default Lit
 }
 nodes [
  n1 = GetMaterialAttributes(None=MaterialFunctionCall() {materialFunction=/VehicleCosmetics/Materials/VehicleParents_Version2/Material_Functions/MF_Vehicle_Body_Base_v2}.Result) {attributeGetTypes [#1]}
  n2 = TextureSample() {texture=/VehicleCosmetics/Bodies/Oblong/Textures/Body/T_Oblong_Skin_Accent}
  n3 = BlendMaterialAttributes(
   A=@n1.MaterialAttributes
   B=MaterialFunctionCall() {materialFunction=&/MF_Skin_Oblong_CuriousTuna01}.Result
   Alpha=Subtract(A=@n2.R, B=@n2.R)
  )
  MaterialFunctionCall!(In=MaterialFunctionCall!(In=MaterialFunctionCall!(MaterialAttributes=@n3, INCarbonFIberMask=@n2.G) {materialFunction=/VehicleCosmetics/Materials/MaterialFunctions/MF_Apply_CarbonFiber}.SetMaterialAttributes) {materialFunction=/VehicleCosmetics/Materials/MaterialFunctions/MF_BodyNormalLOD}.Result) {materialFunction=/VehicleCosmetics/Materials/MaterialFunctions/MF_Vehicle_FNPostFX}
  MaterialFunctionCall(In=MaterialFunctionCall() {materialFunction=&MF_BaseVehicle_Damage}.Result) {materialFunction=&MF_VehicleDeRez_Transition}
  NamedRerouteDeclaration(None=@n1.ClearCoatBottomNormal) {name=BodyBase_BottomNormal, nodeColor=<rgba 0 1 0.3672 1>}
 ]
}
